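# Sound cue for the Storm King monster's charging vocal effort, routed to the boss fight VOX sound class.
# A Random node picks one of four equally weighted windup voice waves and feeds it through a Modulator at the root.
# Attenuation overrides turn off volume falloff despite the referenced Storm King VOX attenuation asset, and a concurrency override prevents new instances at the limit.
# The debug flag is left enabled.
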SoundCue /Game/Sounds/Fort_Monster_Storm_King/Efforts/Storm_King_Charging_Vox_Cue
root=Modulator_0
nodes [
 {id=Modulator_0, children [Random_0]}
 {id=Random_0, children [WavePlayer_0,WavePlayer_1,WavePlayer_2,WavePlayer_3], properties {weights [1,1,1,1]}}
 {id=WavePlayer_0, properties {soundWaveAssetPtr=&/VOX_StormKing_Windup_01}}
 {id=WavePlayer_1, properties {soundWaveAssetPtr=&/VOX_StormKing_Windup_02}}
 {id=WavePlayer_2, properties {soundWaveAssetPtr=&/VOX_StormKing_Windup_03}}
 {id=WavePlayer_3, properties {soundWaveAssetPtr=&/VOX_StormKing_Windup_04}}
]
properties {
 attenuationSettings=/Game/Sounds/Attenuation/Fort_StormKingVox_Att
 bDebug=true
 bOverrideAttenuation=true
 bOverrideConcurrency=true
 maxDistance=2097152
 soundClassObject=/Game/Sounds/FortSoundClassesAndModes/Fort_BossFight_SoundClass/Fort_BossFight_VOX
 volumeMultiplier=1
 attenuationOverrides {bAttenuate=false, distanceAlgorithm=NaturalSound, falloffDistance=5000, lPFRadiusMin=17, stereoSpread=1000, attenuationShapeExtents {x=1000}}
 concurrencyOverrides {resolutionRule=PreventNew}
}
